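# Physics asset for the SharkAnimBite skeletal mesh in the Shark Fangs set.
# It holds a single capsule body on the Hip bone, offset and rotated to approximate the mesh.
# There are no constraints, so it provides one simple collision and bounds shape rather than an articulated ragdoll.
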PhysicsAsset /DealershipSharedLibrary/Assets/Sets/Shark_Fangs/Anim/SharkAnimBite_SK_PhysicsAsset
bodies {Hip [{shapeType=Capsule, length=27.9236, radius=12.3995, center=<-4.7224 -2.9934 0.0002>, rotation=<rot 0 94.3199 -89.9973>}]}
constraints []
