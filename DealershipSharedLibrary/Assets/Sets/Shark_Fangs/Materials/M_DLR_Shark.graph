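# M_DLR_Shark is a masked, lit skeletal-mesh material for the shark in the Shark Fangs set.
# FangsDiffuse_T red and green channels mask tintable teeth and eye/mouth colors, with teeth also declared as emission.
# A Fresnel rim, faded by vertical view angle, multiplies SharkColor routed through the painted-variants function.
# Cloud noise thresholded against MeshEmit VectorColor builds a dissolve-style opacity mask and a colored noise band.
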Material /DealershipSharedLibrary/Assets/Sets/Shark_Fangs/Materials/M_DLR_Shark
graph {
 properties {
  bAutomaticallySetUsageInEditor=false
  bUsedWithMeshDeformer=true
  bUsedWithSkeletalMesh=true
  blendMode=BLEND_Masked
  usedShadingModels=Default Lit
 }
 nodes [
  Constant4Vector() {constant=<rgba 0.5052 0.094 0 1>}
  Constant4Vector() {constant=<rgba 0.5052 0.1369 0.3021 1>}
  n1 = NamedRerouteUsage()
  LinearInterpolate(
   A=Add(A=Multiply(A=@n1), B=NamedRerouteUsage())
   Alpha=MaterialFunctionCall(In=Fresnel() {baseReflectFraction=0.1, exponent=4}, Contrast=Constant() {r=0.5}) {materialFunction=&CheapContrast}.Result
  ) {constA=1, constB=0}
  MaterialFunctionCall(Emissive=NamedRerouteUsage()) {materialFunction=&MF_FX_EmissiveAdjustments}
  n2 = TextureSample() {texture=/DealershipSharedLibrary/Assets/Sets/Shark_Fangs/Textures/FangsDiffuse_T}
  n3 = Multiply(A=@n2.R, B=VectorParameter() {parameterName=TeethSkinColor, defaultValue=<rgba 1 1 0.88 1>}.RGB)
  NamedRerouteDeclaration(None=@n3) {name=Emission, nodeColor=<rgba 0 1 0.2031 1>}
  n4 = DotProduct(
   A=MaterialFunctionCall() {materialFunction=&CameraDirectionVector}.Result
   B=Constant3Vector() {constant=<rgba 0 0 1 0>}
  )
  n5 = ScalarParameter() {defaultValue=1, parameterName=Edge Fade}
  n6 = Multiply(A=Clamp(None=Power(Base=@n4, Exp=@n5)), B=Clamp(None=Fresnel() {exponent=3}))
  n7 = LinearInterpolate(A=@n6, B=Multiply(A=@n6)) {constAlpha=0.6}
  n8 = ScalarParameter() {defaultValue=15, parameterName=Diffuse Boost}
  NamedRerouteDeclaration(None=Power(Base=Multiply(A=@n7, B=@n8) {constB=5}, Exp=ScalarParameter() {defaultValue=1.15, parameterName=Diffuse Contrast})) {name=Edge Out, nodeColor=<rgba 0.9219 0 1 1>}
  n9 = VectorParameter() {parameterName=SharkColor, defaultValue=<rgba 0.126 0.2 0.225 1>}
  NamedRerouteDeclaration(None=Multiply(
    A=MaterialFunctionCall!(Primary Color=@n9.RGBA, Tertiary Color, "Secondary  Color") {materialFunction=/DealershipSharedLibrary/FX/Materials/MaterialFunctions/MF_DLR_PaintedVariantsData}.Primary RGB
    B=NamedRerouteUsage()
   )) {name=Fresnel Color Out, nodeColor=<rgba 0 1 0.6016 1>}
  n10 = Multiply(A=@n2.G, B=VectorParameter() {parameterName=EyesMouthColor, defaultValue=<rgba 0.84 0.03 0.04 1>}.RGB)
  NamedRerouteDeclaration(None=Clamp(None=Add(A=@n3, B=@n10))) {name=Detail Colors, nodeColor=<rgba 1 0.2812 0 1>}
  n11 = Clamp(None=TextureSample() {texture=/DealershipSharedLibrary/FX/Textures/Noise/T_VCVFX_CloudDistort_A}.G)
  n12 = LinearInterpolate(Alpha=VectorParameter() {parameterName=MeshEmit VectorColor, defaultValue=<rgba 1 1 1 1>}.RGBA) {constB=1.5}
  n13 = Ceil(None=Subtract(A=Add(A=@n11) {constB=0.75}, B=@n12))
  NamedRerouteDeclaration(None=Multiply(A=Clamp(None=@n13), B=VectorParameter() {parameterName=MeshEmit VectorColor, defaultValue=<rgba 1 1 1 1>}.RGB)) {name=Noise, nodeColor=<rgba 0 0.0391 1 1>}
  n14 = Add(A=@n11) {constB=0.35}
  NamedRerouteDeclaration(None=Clamp(None=Subtract(A=@n12, B=@n14))) {name=Opacity, nodeColor=<rgba 0 1 0.4375 1>}
 ]
}
